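# Sidebar onboarding coach mark, based on FortSidebarOnboardTooltipWidget, that points players to the voice chat mute, block, report and settings controls.
# OnShowTooltip activates the embedded WBP_UIKit_CoachMark.
# "Got it" (after HandleOtherActionClicked) or OnEndTooltip calls NotifyCloseAnimationFinished, then deactivates the coach mark and widget.
# SetCoachMarkPlacement sets arrow position and toggles the MouseBlocker image's visibility for Inside/Outside, with a design-time preview in PreConstruct.
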WidgetBlueprint /FortniteCoreUI/UI/Social/Chat/TextChat/Coachmarks/WBP_VoiceChat_ControlsMoved_CoachMark
graphs [
 EventGraph {nodes [
   "Button Other Action Clicked (WBP_UIKit_CoachMark)"() {is=title K2Node_ComponentBoundEvent}
   > n1 = HandleOtherActionClicked(self:Fort Sidebar Onboard Tooltip Widget Object Reference)
   NotifyCloseAnimationFinished(execute=[@n1.then, @n3.then], self:Fort Sidebar Onboard Tooltip Widget Object Reference)
   > DeactivateWidget(self=Get WBP_UIKit_CoachMark() -> (WBP_UIKit_CoachMark:WBP UIKit Coach Mark Object Reference))
   > DeactivateWidget(self=Self-Reference())
   n2 = Event PreConstruct()
   > Branch(Condition=@n2.IsDesignTime)
   > SetCoachMarkPlacement!(
    execute
    self
    CheckMarkPlacement:EFortSidebarCoachMarkPlacement Enum=Get PreviewCoachMarkPlacement() -> (PreviewCoachMarkPlacement:EFortSidebarCoachMarkPlacement Enum=Outside)
   ) -> (then)
   Event OnShowTooltip() {is=title K2Node_Event}
   > ActivateWidget(self=Get WBP_UIKit_CoachMark())
   n3 = Event OnEndTooltip() {is=title K2Node_Event}
   Event OnRemovedFromFocusPath!() -> (OutputDelegate, then, InFocusEvent:Focus Event Structure) {is=title K2Node_Event}
  ]}
 SetCoachMarkPlacement {nodes [
   n1 = SetCoachMarkPlacement!() -> (then, CheckMarkPlacement:EFortSidebarCoachMarkPlacement Enum=Outside) {is=title K2Node_FunctionEntry}
   > SetVisibility(
    self=Get MouseBlocker_PreventsHoversOnSocialListItems() -> (MouseBlocker_PreventsHoversOnSocialListItems:Image Object Reference)
    InVisibility=Select!(Outside:ESlateVisibility Enum=SelfHitTestInvisible, Inside:ESlateVisibility Enum=Collapsed, Index:EFortSidebarCoachMarkPlacement Enum=@n1) -> (ReturnValue)
   )
   > SetArrowPosition!(
    execute
    self:WBP UIKit Coach Mark Base Object Reference=Get WBP_UIKit_CoachMark()
    ArrowPosition:E_UI_OutBoardPosition Enum=Select(Outside:E_UI_OutBoardPosition Enum=NewEnumerator10, Inside:E_UI_OutBoardPosition Enum=NewEnumerator0, Index=@n1) -> (ReturnValue:E_UI_OutBoardPosition Enum)
   ) -> (then)
  ]}
]
variables [PreviewCoachMarkPlacement {type=EFortSidebarCoachMarkPlacement, default=Outside}]
widgets {
 info {inheritedWidgetCount=2, parentClass=/Script/FortniteUI.FortSidebarOnboardTooltipWidget, rootWidgetClass=&Overlay, widgetCount=5}
 widgets [
  Overlay_0 {widgetClassPath=&Overlay, widget=Overlay_0}
  MouseBlocker_PreventsHoversOnSocialListItems {
   widgetClassPath=&Image
   bIsVariable=true
   parent=Overlay_0
   widget=MouseBlocker_PreventsHoversOnSocialListItems
   properties {brush {
     bIsDynamicallyLoaded=false
     drawAs=NoDrawType
     imageType=NoImage
     mirroring=NoMirror
     resourceName=None
     resourceObject=null
     tiling=NoTile
     imageSize=<32 32>
     margin=<ltrb 0 0 0 0>
     outlineSettings {
      bUseBrushTransparency=false
      roundingType=HalfHeightRadius
      width=0
      color {colorUseRule=UseColor_Specified, specifiedColor=<rgba 0 0 0 0>}
      cornerRadii=<0 0 0 0>
     }
     tintColor {colorUseRule=UseColor_Specified, specifiedColor=<rgba 1 1 1 1>}
     uVRegion {bIsValid=false, max=<0 0>, min=<0 0>}
    }}
   slot {class=OverlaySlot, properties {horizontalAlignment=HAlign_Fill, verticalAlignment=VAlign_Fill}}
  }
  WBP_UIKit_CoachMark {
   widgetClassPath=/FortUILibrary/UIKit/CoachMark/WBP_UIKit_CoachMark
   bIsVariable=true
   parent=Overlay_0
   widget=WBP_UIKit_CoachMark
   properties {
    arrowPosition=Right_Up
    navigation=&self:WidgetTree.WBP_UIKit_CoachMark.WidgetNavigation_0
    shouldShowDismissButton=false
    textDescription="Easily mute your mic or block all chat audio right here. Report chat or change your settings from the ••• menu."
    textHeader=Control your voice chats
    textOtherActionButton=Got it
    widget Width=580
   }
   slot {class=OverlaySlot}
  }
  Dismiss_Button {widgetClassPath=&CommonButtonLegacy, bInherited=true, widget=null}
  OtherAction_Button {widgetClassPath=&CommonButtonLegacy, bInherited=true, widget=null}
 ]
}
