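# Translucent UI material for the Sidebar's solo party icon, drawing a grey friends-list glyph on a faint white rounded backing.
# The MSDF texture is sampled with center-scaled UVs, converted to an SDF and anti-aliased into a reusable Icon Mask.
# That mask blends white to grey for color, and lifts a dimmed pixel-radius rounded-rectangle mask to full value, apparently for opacity.
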Material /Sidebar/Materials/Icons/M_UI_Sidebar_SoloParty
graph {
 properties {bAutomaticallySetUsageInEditor=false, blendMode=BLEND_Translucent, materialDomain=MD_UI, usedShadingModels=Default Lit}
 nodes [
  n1 = Constant4Vector() {constant=<rgba 12 12 12 12>}
  n2 = Constant() {r=1}
  n3 = MaterialFunctionCall() {materialFunction=&GetUserInterfaceUV}
  n4 = Constant() {r=0.6}
  n5 = MaterialFunctionCall!(MSDF RGB=TextureSample(UVs=MaterialFunctionCall(UVs=@n3.Normalized UV, Texture Scale=@n4) {materialFunction=&ScaleUVsByCenter}.UVs) {samplerType=SAMPLERTYPE_LinearColor, texture=/Sidebar/Materials/Icons/Textures/T_UI_Icon_Sidebar_FriendsList_MSDF}.RGB) {materialFunction=/FortUILibrary/MaterialFunctions/Shapes/Modifiers/MF_UI_MSDFtoSDF}
  n6 = Constant() {r=0.5}
  n7 = Constant() {r=0.01}
  LinearInterpolate(
   A=Multiply(A=MaterialFunctionCall(Corner Radius=@n1.RGBA, Radius in Pixels=@n2) {materialFunction=&MF_UI_CoreUI_RoundedMask}.Shape Mask) {constB=0.05}
   Alpha=NamedRerouteDeclaration(None=MaterialFunctionCall(DistanceField=@n5.SDF, Threshold=@n6, Smoothing=@n7) {materialFunction=&MF_UI_AntiAliasing}.Outer) {name=Icon Mask, nodeColor=<rgba 0.7578 0 1 1>}
  )
  LinearInterpolate(A=Constant3Vector() {constant=<rgba 1 1 1 0>}, B=Constant3Vector() {constant=<rgba 0.5776 0.5776 0.5776 1>}, Alpha=NamedRerouteUsage())
 ]
}
