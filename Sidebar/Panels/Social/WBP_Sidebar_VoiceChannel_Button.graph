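# Sidebar button widget for a voice channel in the social panel, derived from WBP_Sidebar_Button_Base.
# A grid holds a named left slot beside a column with group label and status texts, both truncated by ellipsis.
# PreConstruct applies GroupLabelText through SetGroupLabel, which stores and displays it; SetGroupStatus sets the status line.
# SetChannelButtonActive shows the active channel at full opacity and dims inactive ones via the root grid.
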WidgetBlueprint /Sidebar/Panels/Social/WBP_Sidebar_VoiceChannel_Button
graphs [
 EventGraph {nodes [
   n1 = Event PreConstruct()
   > Parent: PreConstruct(IsDesignTime=@n1.IsDesignTime)
   > SetGroupLabel!(execute, self, LabelText:Text=Get GroupLabelText() -> (GroupLabelText:Text)) -> (then)
  ]}
 SetChannelButtonActive {nodes [
   n1 = SetChannelButtonActive!() -> (then, IsCurrentlyTheActiveChannel:bool=~) {is=title K2Node_FunctionEntry}
   > n2 = Branch(Condition=@n1)
   n3 = Get RootGrid() -> (RootGrid:Grid Panel Object Reference)
   > SetRenderOpacity(self=@n3, InOpacity=1)
   SetRenderOpacity(execute=@n2.else, self=@n3, InOpacity=0.65)
  ]}
 SetGroupLabel {nodes [
   n1 = SetGroupLabel!() -> (then, LabelText:Text) {is=title K2Node_FunctionEntry}
   > n2 = Set GroupLabelText(GroupLabelText:Text=@n1) -> (Output_Get:Text)
   > "SetText (Text)"(self=Get Text_Group() -> (Text_Group:Common Text Object Reference), InText=@n2)
  ]}
 SetGroupStatus {nodes [
   n1 = SetGroupStatus!() -> (then, StatusText:Text) {is=title K2Node_FunctionEntry}
   > "SetText (Text)"(self=Get Text_Status() -> (Text_Status:Common Text Object Reference), InText=@n1)
  ]}
]
variables [GroupLabelText {type=string, default=Group Label}]
widgets {
 info {
  inheritedWidgetCount=0
  namedSlotCount=2
  parentClass=/Sidebar/Buttons/WBP_Sidebar_Button_Base.WBP_Sidebar_Button_Base_C
  rootWidgetClass=null
  widgetCount=8
 }
 widgets [
  Image_28 {widgetClassPath=&Image, bIsVariable=true, widget=Image_28}
  RootGrid {widgetClassPath=&GridPanel, bIsVariable=true, widget=RootGrid, properties {columnFill [0,1], rowFill [1]}}
  VB_RightColumn {
   widgetClassPath=&VerticalBox
   parent=RootGrid
   widget=VB_RightColumn
   slot {class=GridSlot, properties {column=1, columnSpan=1, verticalAlignment=VAlign_Center, padding=<ltrb 0 12 12 12>}}
  }
  HorizontalStackBox_Name {
   widgetClassPath=&StackBox
   parent=VB_RightColumn
   widget=HorizontalStackBox_Name
   slot {class=VerticalBoxSlot, properties {verticalAlignment=VAlign_Top, padding=<ltrb 0 0 0 3>}}
  }
  SizeBox_TextClipping {
   widgetClassPath=&SizeBox
   bIsVariable=true
   parent=HorizontalStackBox_Name
   widget=SizeBox_TextClipping
   properties {clipping=ClipToBounds}
   slot {class=StackBoxSlot, properties {size {sizeRule=Fill, value=1}}}
  }
  Text_Group {
   widgetClassPath=&CommonTextBlock
   bIsVariable=true
   parent=SizeBox_TextClipping
   widget=Text_Group
   properties {
    mobileFontSizeMultiplier=1.4
    text="Group "
    textOverflowPolicy=Ellipsis
    colorAndOpacity {colorUseRule=UseColor_Specified, specifiedColor=<rgba 0.7913 0.7913 0.8228 1>}
    font {
     bForceMonospaced=false
     bMaterialIsStencil=false
     fontMaterial=null
     fontObject=&HeadingNow
     letterSpacing=10
     monospacedWidth=1
     size=13.5
     skewAmount=0
     typefaceFontName=75 Medium
     outlineSettings {
      bApplyOutlineToDropShadows=false
      bMiteredCorners=false
      bSeparateFillAlpha=false
      outlineMaterial=/FortUILibrary/Systems/ModularButton/MaterialInstances/MI_UIKit_TextOutline
      outlineSize=0
      outlineColor=<rgba 0 0 0 1>
     }
    }
    shadowOffset=<0 0>
   }
   slot {class=SizeBoxSlot, properties {verticalAlignment=VAlign_Center}}
  }
  Text_Status {
   widgetClassPath=&CommonTextBlock
   bIsVariable=true
   parent=VB_RightColumn
   widget=Text_Status
   properties {
    clipping=ClipToBounds
    mobileFontSizeMultiplier=1.6
    text=Dynamic Status
    textOverflowPolicy=Ellipsis
    colorAndOpacity {colorUseRule=UseColor_Specified, specifiedColor=<rgba 1 1 1 0.5>}
    font {
     bForceMonospaced=false
     bMaterialIsStencil=false
     fontMaterial=null
     fontObject=&HeadingNow
     letterSpacing=0
     monospacedWidth=1
     size=11.25
     skewAmount=0
     typefaceFontName=74 Regular
     outlineSettings {
      bApplyOutlineToDropShadows=false
      bMiteredCorners=false
      bSeparateFillAlpha=false
      outlineMaterial=&MI_UIKit_TextOutline
      outlineSize=0
      outlineColor=<rgba 0 0 0 1>
     }
    }
    shadowOffset=<0 0>
   }
   slot {class=VerticalBoxSlot, properties {verticalAlignment=VAlign_Bottom, size {sizeRule=Fill, value=1}}}
  }
  Slot_Left {
   widgetClassPath=&NamedSlot
   bIsVariable=true
   parent=RootGrid
   widget=Slot_Left
   slot {class=GridSlot, properties {horizontalAlignment=HAlign_Left, verticalAlignment=VAlign_Center, padding=<ltrb 12 12 12 12>}}
  }
 ]
}
